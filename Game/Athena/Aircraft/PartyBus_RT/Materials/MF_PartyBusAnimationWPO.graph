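# Material function that builds World Position Offset animation for the Party Bus aircraft.
# Static switches layer balloon sway with radial falloff, Pivot Painter per-element wind wiggle and constant spin, vibration, and a sine wave, mostly masked by vertex color.
# A slow whole-body rotation around the actor-relative center is added last, and AnimationFadeIn can scale it and the wiggle.
# Mobile zeroes the layered motion unless switches allow constant rotation or the full stack, excluding overall rotation.
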MaterialFunction /Game/Athena/Aircraft/PartyBus_RT/Materials/MF_PartyBusAnimationWPO
graph {nodes [
  Constant()
  Constant()
  n1 = Multiply(A=Time(), B=ScalarParameter() {defaultValue=806.4869, parameterName=SimpleSineWaveSpeed})
  Frac(None=@n1)
  n2 = Multiply(A=Time(), B=ScalarParameter() {defaultValue=5, parameterName=SpeedVibration})
  n3 = Sine(None=@n2)
  n4 = AppendVector(A=Sine(None=@n2), B=Cosine(None=@n2))
  n5 = Multiply(A=Normalize(VectorInput=AppendVector(A=@n3, B=@n4)), B=ScalarParameter() {defaultValue=0.5, parameterName=VibrationIntensity})
  n6 = ChannelMaskParameter(None=VertexColor()) {maskChannel=Blue, parameterName=Vibration_ChannelMask, defaultValue=<rgba 0 0 1 0>}
  n7 = Time()
  n8 = Multiply(A=@n7, B=ScalarParameter() {defaultValue=0.1714, parameterName=SpeedBallon_1})
  n9 = Sine(None=@n8)
  n10 = Cosine(None=@n8)
  n11 = Normalize(VectorInput=AppendVector(A=Constant() {r=1}, B=AppendVector(A=@n9, B=@n10)))
  n12 = ScalarParameter() {defaultValue=0.3429, parameterName=SpeedBallon}
  n13 = Multiply(A=Sine(None=Multiply(A=@n7, B=@n12)), B=ScalarParameter() {defaultValue=0.01, parameterName=Ballonrotation})
  n14 = WorldPosition() {worldPositionShaderOffset=WPT_ExcludeAllShaderOffsets}
  n15 = Add(A=VectorParameter() {parameterName=RotationCenter, defaultValue=<rgba 0 0 336.121 1>}.RGB, B=ActorPositionWS())
  n16 = WorldPosition() {worldPositionShaderOffset=WPT_ExcludeAllShaderOffsets}
  n17 = Divide(
   A=Add(A=Distance(A=@n16.XYZ, B=@n15), B=ScalarParameter() {defaultValue=-500, parameterName=RadiusBallonFalloff_Start})
   B=ScalarParameter() {defaultValue=801.6973, parameterName=RadiusBallonFalloff}
  )
  n18 = StaticSwitchParameter(
   True=Multiply(A=RotateAboutAxis(NormalizedRotationAxis=@n11, RotationAngle=@n13, PivotPoint=@n15, Position=@n14.XYZ), B=Clamp(None=@n17))
   False=Constant3Vector()
  ) {parameterName=BallonAnimation}
  n19 = Transform(None=VectorParameter() {parameterName=RotaionAxis, defaultValue=<rgba 0 0 1 1>}.RGB) {transformSourceType=TRANSFORMSOURCE_Local}
  n20 = Time()
  n21 = MaterialFunctionCall() {materialFunction=/Engine/Functions/Engine_MaterialFunctions02/PivotPainter/PivotPainter_PerObjectData}
  n22 = Add(A=@n20, B=Multiply(A=@n21.Random Value Per Element, B=ScalarParameter() {defaultValue=1.5613, parameterName=OffsetVariation}))
  n23 = ScalarParameter() {defaultValue=2.3369, parameterName=WindRotation_Speed_1}
  n24 = Add(A=Multiply(A=@n21.Random Value Per Element, B=ScalarParameter() {defaultValue=2.2338, parameterName=OffsetVariation_1}), B=@n20)
  n25 = ScalarParameter() {defaultValue=2.7346, parameterName=WindRotation_Speed}
  n26 = Add(A=Sine(None=Multiply(A=@n22, B=@n23)), B=Sine(None=Multiply(A=@n24, B=@n25)))
  n27 = ScalarParameter() {defaultValue=0.005, parameterName=WindRotation_strength}
  n28 = Multiply(
   A=RotateAboutAxis(
    NormalizedRotationAxis=Normalize(VectorInput=@n19)
    RotationAngle=Multiply(A=@n26, B=@n27)
    PivotPoint=@n21.Pivot Position
    Position=@n14.XYZ
   )
   B=ChannelMaskParameter(None=VertexColor()) {maskChannel=Green, parameterName=RotationWiggle_ChannelMask, defaultValue=<rgba 0 1 0 0>}.R
  )
  n29 = ScalarParameter() {parameterName=AnimationFadeIn}
  n30 = Multiply(A=@n28, B=@n29)
  n31 = StaticSwitchParameter(True=Add(A=@n18, B=StaticSwitchParameter(True=@n30, False=@n28) {parameterName=UseAnimationFadein_1}), False=@n18) {defaultValue=true, parameterName=UseRotationWiggle}
  n32 = Transform(None=VectorParameter() {parameterName=RotaionAxis_ConstantRot, defaultValue=<rgba 0.1429 1 0 1>}.RGB) {transformSourceType=TRANSFORMSOURCE_Local}
  n33 = Time()
  n34 = ScalarParameter() {defaultValue=2.3519, parameterName=ConstantRotationSpeed}
  n35 = VectorParameter() {parameterName=ConstantRotationOffset, defaultValue=<rgba 0 0 0 1>}
  n36 = Multiply(
   A=RotateAboutAxis(
    NormalizedRotationAxis=Normalize(VectorInput=@n32)
    RotationAngle=Multiply(A=@n33, B=@n34)
    PivotPoint=MaterialFunctionCall() {materialFunction=&PivotPainter_PerObjectData}.Pivot Position
    Position=Add(A=@n35.RGB, B=@n14.XYZ)
   )
   B=ChannelMaskParameter(None=VertexColor()) {parameterName=ConstantRotation_ChannelMask}.R
  )
  n37 = StaticSwitchParameter(True=Add(A=@n31, B=@n36), False=@n31) {parameterName=UseConstantRotation}
  n38 = StaticSwitchParameter(True=Add(A=Multiply(A=@n5, B=@n6.R), B=@n37), False=@n37) {defaultValue=true, parameterName=UseVibration}
  n39 = Transform(None=Subtract(A=WorldPosition() {worldPositionShaderOffset=WPT_ExcludeAllShaderOffsets}.XYZ, B=ActorPositionWS())) {transformSourceType=TRANSFORMSOURCE_World, transformType=TRANSFORM_Local}
  n40 = Divide(A=Add(A=ComponentMask(None=@n39) {g=true}, B=@n1), B=ScalarParameter() {defaultValue=730.2227, parameterName=SimpleSineWaveLength})
  n41 = Multiply(
   A=Multiply(A=Sine(None=@n40), B=ScalarParameter() {defaultValue=26.8963, parameterName=SimpleSineWaveAmplitude})
   B=VectorParameter() {parameterName=SimpleSineOffsetDirection, defaultValue=<rgba 0 0 1 1>}.RGB
  )
  n42 = VertexColor()
  n43 = StaticSwitchParameter(True=Add(A=@n38, B=Multiply(A=@n41, B=@n42.R)), False=@n38) {parameterName=UseSimpleSineWaveWiggle}
  n44 = ShadingPathSwitch(
   Default=@n43
   Mobile=StaticSwitchParameter(True=StaticSwitchParameter(True=@n43, False=@n36) {parameterName=AllowConstantRotationVibrationOnMobile}, False=Constant3Vector()) {parameterName=AllowConstantRotationOnMobile}
  ) {desc=Migrated from FeatureLevelSwitch}
  n45 = Multiply(A=Time(), B=ScalarParameter() {defaultValue=0.1714, parameterName=SpeedBallon_2})
  n46 = Sine(None=@n45)
  n47 = AppendVector(A=Sine(None=@n45), B=Cosine(None=@n45))
  n48 = Multiply(A=Time(), B=ScalarParameter() {defaultValue=0.0892, parameterName=OverallRotationSpeed})
  n49 = RotateAboutAxis(
   NormalizedRotationAxis=Normalize(VectorInput=AppendVector(A=@n46, B=@n47))
   RotationAngle=Multiply(A=Sine(None=@n48), B=ScalarParameter() {defaultValue=0.01, parameterName=OverallRotation})
   PivotPoint=@n15
   Position=@n14.XYZ
  )
  n50 = StaticSwitchParameter(True=Multiply(A=@n29, B=@n49), False=@n49) {parameterName=UseAnimationFadein}
  FunctionOutput(None=StaticSwitchParameter(True=Add(A=@n44, B=@n50), False=@n44) {defaultValue=true, parameterName=UseOverallRotation})
  MaterialFunctionCall() {materialFunction=&PivotPainter_PerObjectData}
  n51 = AppendVector(A=TextureCoordinate() {coordinateIndex=2}, B=ComponentMask(None=TextureCoordinate() {coordinateIndex=3}) {r=true})
  n52 = Constant3Vector() {constant=<rgba 0 -1 0 0>}
  Normalize(VectorInput=TransformPosition(None=Add(A=@n51, B=@n52)) {transformType=TRANSFORMPOSSOURCE_World})
  VertexColor()
  WorldPosition()
  WorldPosition()
 ]}
