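# Master material for the Party Bus aircraft with anime shading, built on material attributes for skeletal meshes.
# Shared character PreFX feeds MF_AnimeShading, whose result runs through character PostFX.
# Party Bus animation WPO is then added to the world position offset.
# The UseWPO_ClothAnimation static switch adds cloth WPO, masked by a vertex color channel (blue by default), and blends the cloth normal into the base normal.
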Material /Game/Athena/Aircraft/PartyBus_RT/Materials/Anime/M_PartyBus_Anim_MASTER_AnimeShading1
graph {
 properties {
  bUseMaterialAttributes=true
  bUsedWithMeshDeformer=true
  bUsedWithSkeletalMesh=true
  pixelDepthOffsetMode=PDOM_Legacy
  usedShadingModels=Default Lit
 }
 nodes [
  n1 = MaterialFunctionCall() {materialFunction=&BaseFN_CharacterFunction_PreFX}
  n2 = MaterialFunctionCall!(
   Material=@n1.Result
   LitDiffuseOverride
   SSC_AlphaOverride
   Tint Mask
   DF InkLine Visibility Mask
   DF Inkline Injection
   DF Inkline Color Override
   DF Inkline Color Override Mask
   CustomUVs
  ) {materialFunction=/Game/Packages/Fortress_SharedMaterials/Anime_Master_Material/MF_AnimeShading}
  MaterialFunctionCall!(Material=@n2.Result) {materialFunction=/Game/Packages/Fortress_SharedMaterials/Anime_Master_Material/MF_ToonOutline_Mobile}
  n3 = MaterialFunctionCall(AO=@n1.AO, Distance Blend=@n1.AO, Fresnel=@n1.AO, Gmap=@n1.AO, Pre Skinned Local Position=@n1.AO, Material=@n2.Result) {materialFunction=&BaseFN_CharacterFunction_PostFX}
  n4 = GetMaterialAttributes(None=@n3.Result) {attributeGetTypes [#1, #2]}
  n5 = MaterialFunctionCall() {materialFunction=/Game/Athena/Aircraft/PartyBus_RT/Materials/MF_PartyBusAnimationWPO}
  n6 = MaterialFunctionCall!(
   UVs
   Pan Speed
   WPO_ClothOffsetDirection
   WPO_ClothNormalIntensity
   WPO_ClothOffsetIntensity
   "Motion Mask (Defaults to the Red Vertex Color)"=ChannelMaskParameter(None=VertexColor()) {group=WPO_ClothAnimation, maskChannel=Blue, parameterName=ClothMask, defaultValue=<rgba 0 0 1 0>}.R
   WPO_ClothWPOWindTextureOffset
   Pan Direction
  ) {materialFunction=/Game/Packages/Fortress_SharedMaterials/Base_Material_Functions/Cloth}
  n7 = Add(A=@n5.Result, B=@n6.Additive WPO)
  SetMaterialAttributes!(
   MaterialAttributes=@n4.MaterialAttributes
   World Position Offset=Add(
    A=@n4.World Position Offset
    B=StaticSwitchParameter(True=@n7, False=@n5.Result) {group=WPO_ClothAnimation, parameterName=UseWPO_ClothAnimation}
   )
   Normal=StaticSwitchParameter(
    True=MaterialFunctionCall(BaseNormal=@n6.Normal, AdditionalNormal=@n4.Normal) {materialFunction=&BlendAngleCorrectedNormals}.Result
    False=@n4.Normal
   ) {group=WPO_ClothAnimation, parameterName=UseWPO_ClothAnimation}
  ) {attributeSetTypes [#2, #1]}
 ]
}
