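# Material function that draws Lexa's stylized anime eyes over incoming material attributes and blends them to robot eyes via a transformation curve parameter.
# Look-pose parameters offset clamped, scaled pupil UVs, while camera-to-sunlight angle and horizontal gaze rotate two sheen layers.
# The masked result replaces base color and emissive over darkened whites, with constant roughness, specular and metallic.
# Sphere-mask highlight, ambient motion and timed-transition branches are left unconnected.
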MaterialFunction /Game/Characters/Player/Female/Medium/Bodies/F_MED_Lexa/Materials/MF_AnimeEyes
graph {nodes [
  n1 = Panner(Coordinate=ComponentMask(None=MaterialFunctionCall!(Coordinate) {materialFunction=/Game/Effects/Niagara/MasterMaterials/Functions/UVs/PolarCoordinates}.Result) {g=true, r=true}) {speedX=1}
  n2 = ComponentMask(None=VectorParameter() {parameterName=Sheen_Spot_01_Pos, defaultValue=<rgba -0.1 -1 0 1>}.RGB) {g=true, r=true}
  n3 = ScalarParameter() {defaultValue=0.1, parameterName=Sheen_01_Radius}
  n4 = ScalarParameter() {defaultValue=1, parameterName=Sheen_01_Hardness}
  n5 = Multiply(A=TextureCoordinate()) {constB=2}
  n6 = MaterialFunctionCall(
   UVs=Rotator(Coordinate=Subtract(A=@n5)) {centerX=0, centerY=0}
   Texture Scale=ComponentMask(None=VectorParameter() {parameterName=Sheen_02_Scale, defaultValue=<rgba 0.5 1 0 1>}.RGB) {g=true, r=true}
  ) {materialFunction=&ScaleUVsByCenter}
  n7 = ComponentMask(None=VectorParameter() {parameterName=Sheen_Spot_02_Pos, defaultValue=<rgba 0 0.2 0 1>}.RGB) {g=true, r=true}
  n8 = ScalarParameter() {defaultValue=0.1, parameterName=Sheen_02_Radius}
  n9 = ScalarParameter() {defaultValue=1, parameterName=Sheen_02_Hardness}
  n10 = ScalarParameter() {parameterName=look_right_pose}
  n11 = Multiply(A=ScalarParameter() {parameterName=look_left_pose}) {constB=-1}
  n12 = Add(A=Multiply(A=ScalarParameter() {parameterName=look_down_pose}) {constB=-1}, B=ScalarParameter() {parameterName=look_up_pose})
  n13 = AppendVector(
   A=Clamp(
    None=Add(A=@n10, B=@n11)
    Min=ScalarParameter() {defaultValue=-0.25, parameterName=AE_EyeHorizontalThreshold_Min}
    Max=ScalarParameter() {defaultValue=0.25, parameterName=AE_EyeHorizontalThreshold_Max}
   )
   B=Clamp(
    None=Subtract(A=@n12) {constB=0.1}
    Min=ScalarParameter() {defaultValue=-0.25, parameterName=AE_EyeVerticalThreshold_Min}
    Max=ScalarParameter() {defaultValue=0.25, parameterName=AE_EyeVerticalThreshold_Max}
   )
  )
  n14 = Add(A=TextureCoordinate(), B=@n13)
  n15 = MaterialFunctionCall(UVs=@n14, Texture Scale=ScalarParameter() {defaultValue=0.5, parameterName=AE_PupilScale}) {materialFunction=&ScaleUVsByCenter}
  n16 = TextureSampleParameter2D(UVs=@n15.UVs) {parameterName=AE_PupilTexture, texture=/Game/Characters/Player/Female/Medium/Bodies/F_MED_Lexa/Textures/T_Lexa_Pupil}
  n17 = CollectionParameter() {
   collection=&FortniteMaterialParameters
   parameterName=SunlightDirectionalVector
  }
  n18 = ConstantBiasScale(None=DotProduct(
    A=MaterialFunctionCall() {materialFunction=&CameraDirectionVector}.Result
    B=ComponentMask(None=@n17) {b=true, g=true, r=true}
   ))
  n19 = ScalarParameter() {defaultValue=3, parameterName=SheenMovement_Offset}
  n20 = MaterialFunctionCall(Float2=@n13) {materialFunction=&BreakOutFloat2Components}
  n21 = Max(A=@n20.R, B=@n20.R)
  n22 = Multiply(A=Add(A=Add(A=@n18, B=@n19), B=@n21), B=ScalarParameter() {defaultValue=45, parameterName=SheenMovementThreshold}) {constB=0.25, desc=1}
  n23 = ScalarParameter() {parameterName=Sheen_02_Offset}
  n24 = Rotator(Coordinate=@n15.UVs, Time=Multiply(A=Add(A=@n22, B=@n23), B=ScalarParameter() {defaultValue=0.5, parameterName=Sheen_02_Speed}))
  n25 = Rotator(Coordinate=@n15.UVs, Time=@n22)
  n26 = LinearInterpolate(
   A=@n16.RGB
   B=VectorParameter() {parameterName=EyeColorSheen, defaultValue=<rgba 1 0.4678 0.0976 1>}.RGB
   Alpha=Add(
    A=TextureSample(UVs=@n24) {texture=/Game/Characters/Player/Female/Medium/Bodies/F_MED_Lexa/Textures/Lexa_Pupil_RGB}.G
    B=TextureSample(UVs=@n25) {texture=&Lexa_Pupil_RGB}.B
   )
  )
  n27 = Multiply(A=@n16.A, B=@n15.0-1 mask)
  n28 = VectorParameter() {parameterName=RoboEye_BG_Color, defaultValue=<rgba 0 0 0 1>}
  n29 = TextureSampleParameter2D(UVs=@n15.UVs) {parameterName=AE_PupilTexture, texture=&Lexa_Pupil_RGB}
  n30 = LinearInterpolate(A=@n28.RGB, B=VectorParameter() {parameterName=RoboEyeColor, defaultValue=<rgba 1 0.011 0 1>}.RGB, Alpha=@n29.R)
  n31 = Multiply(A=@n29.A, B=@n15.0-1 mask)
  n32 = Clamp(None=Multiply(A=ScalarParameter() {parameterName=Lexa_TransformationCurve}) {constB=2})
  n33 = Multiply(
   A=LinearInterpolate(A=Multiply(A=@n26, B=@n27), B=Multiply(A=@n30, B=@n31), Alpha=@n32)
   B=ScalarParameter() {defaultValue=0.5, parameterName=EyeBrightness}
  )
  Add(A=Max(A=SphereMask(A=@n1, B=@n2, Radius=@n3, Hardness=@n4), B=SphereMask(A=@n6.UVs, B=@n7, Radius=@n8, Hardness=@n9)), B=@n33)
  n34 = Multiply(A=Time(), B=ScalarParameter() {defaultValue=0.125, parameterName=AE_DirectionSwapSpeed})
  n35 = Multiply(A=Add(A=Time()) {constB=0.5}, B=ScalarParameter() {defaultValue=0.125, parameterName=AE_DirectionSwapSpeed})
  n36 = MaterialFunctionCall(X=Round(None=Sine(None=@n34)), Y=Round(None=Sine(None=@n35)), Z=Constant()) {materialFunction=&MakeFloat3}
  n37 = Multiply(A=Constant3Vector() {constant=<rgba 0.25 0 0 0>}, B=@n36.Result)
  n38 = Constant3Vector()
  n39 = Multiply(A=Constant3Vector() {constant=<rgba 0 -0.1 0 0>}, B=@n36.Result)
  n40 = Multiply(A=Time(), B=ScalarParameter() {defaultValue=0.25, parameterName=AE_AmbientMotionSpeed_Global})
  n41 = OneMinus(None=Frac(None=@n40))
  n42 = Divide(A=Ceil(None=Multiply(A=@n41) {constB=3})) {constB=3}
  n43 = Multiply(
   A=ComponentMask(None=MaterialFunctionCall(A=@n37, B=@n38, C=@n39, Alpha=@n42) {materialFunction=&Lerp_3Color}.Result) {g=true, r=true}
   B=ScalarParameter() {parameterName=EnabelMicroStaccata}
  ) {constB=0}
  Add(A=@n14, B=@n43)
  Add(A=@n13, B=@n43)
  ConstantBiasScale(None=Sine(None=@n40))
  Constant()
  Constant()
  Constant()
  n44 = GetMaterialAttributes(None=FunctionInput() {inputName=Material, inputType=FunctionInput_MaterialAttributes}) {attributeGetTypes [#1]}
  n45 = ScalarParameter() {defaultValue=0.5, parameterName=DarkenWhites}
  n46 = LinearInterpolate(A=LinearInterpolate(A=Multiply(A=@n44.EmissiveColor, B=@n45), B=@n28.RGB, Alpha=@n32), B=@n33, Alpha=@n27)
  n47 = SetMaterialAttributes!(
   MaterialAttributes=@n44.MaterialAttributes
   EmissiveColor=@n46
   Roughness=Constant() {r=1}
   Specular=Constant()
   Metallic=Constant()
   BaseColor=@n46
  ) {attributeSetTypes [
    #1
    #2
    #3
    #4
    #5
   ]}
  FunctionOutput(None=@n47)
  LinearInterpolate(
   A=VectorParameter() {parameterName=EyeColorBase, defaultValue=<rgba 1 0.0331 0.011 1>}.RGB
   B=VectorParameter() {parameterName=EyeColorPupil, defaultValue=<rgba 0.3613 0.0091 0.0024 1>}.RGB
   Alpha=TextureSample(UVs=@n15.UVs) {texture=&Lexa_Pupil_RGB}.G
  )
  LinearInterpolate()
  n48 = Time()
  n49 = ScalarParameter() {defaultValue=99999997952, parameterName=LX_GameTimeFromBP}
  n50 = Saturate(None=Divide(A=Subtract(A=@n48, B=@n49), B=ScalarParameter() {defaultValue=1, parameterName=LX_EyeTransitionDuration}))
  LinearInterpolate(A=@n50, B=OneMinus(None=@n50), Alpha=ScalarParameter() {parameterName=LX_TransformationDirection})
  MaterialFunctionCall(
   UVs=@n1
   Texture Scale=ComponentMask(None=VectorParameter() {parameterName=Sheen_01_Scale, defaultValue=<rgba 1 0.5 0 1>}.RGB) {g=true, r=true}
  ) {materialFunction=&ScaleUVsByCenter}
  Multiply(A=Frac(None=Time())) {constB=4}
  Multiply(A=@n21) {constB=-0.5}
  n51 = Multiply(A=Time(), B=ScalarParameter() {defaultValue=0.5, parameterName=AE_AmbientMotionSpeed_Global})
  n52 = Multiply(A=@n51, B=ScalarParameter() {defaultValue=1, parameterName=AE_AmbientMotionSpeed_OutRIght})
  n53 = Multiply(
   A=Add(A=@n51, B=ScalarParameter() {defaultValue=1, parameterName=AE_AmbientMotion_Offset})
   B=ScalarParameter() {defaultValue=0.5, parameterName=AE_AmbientMotionSpeed_UpRIght}
  )
  Multiply(A=AppendVector(A=Sine(None=@n52), B=Sine(None=@n53)), B=ScalarParameter() {defaultValue=0.02, parameterName=AE_AmbientMotion_Threshold})
  n54 = TextureCoordinate()
  Rotator(Coordinate=Subtract(A=Multiply(A=@n54) {constB=2})) {centerX=0, centerY=0}
  Rotator() {centerX=0, centerY=0}
  SphereMask(
   A=Frac(None=@n1)
   B=ComponentMask(None=Constant3Vector() {constant=<rgba 0.25 0.25 0 0>}) {g=true, r=true}
   Radius=Constant() {r=0.1}
   Hardness=@n4
  )
  TextureSampleParameter2D(UVs=@n1) {parameterName=AE_PupilTexture, texture=&T_Lexa_Pupil}
 ]}
