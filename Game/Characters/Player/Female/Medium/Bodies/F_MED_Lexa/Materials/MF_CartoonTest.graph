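# Material function for the Lexa player character that replaces incoming material attributes with sun-driven cartoon lighting.
# Sun direction dotted with the normal, optionally blended with a fake face vector map, indexes a curve atlas ramp that shades base color.
# Optional Phong highlight and sun-facing Fresnel rim are added, then time-of-day adjusted into Emissive, with base color blended from black.
# Time-of-day tint and hair-gradient branches are left unconnected.
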MaterialFunction /Game/Characters/Player/Female/Medium/Bodies/F_MED_Lexa/Materials/MF_CartoonTest
graph {nodes [
  n1 = TextureSampleParameter2D() {
   parameterName=Lexa_FakeFaceVectorMap
   samplerType=SAMPLERTYPE_LinearColor
   texture=/Game/Characters/Player/Female/Medium/Bodies/F_MED_Lexa/Textures/T_Lexa_Test
  }
  n2 = Subtract(A=Multiply(A=@n1.RGB) {constB=2})
  n3 = PixelNormalWS()
  n4 = StaticSwitchParameter(True=LinearInterpolate(A=Transform(None=@n2) {transformSourceType=TRANSFORMSOURCE_Local}, B=@n3) {constAlpha=0.8}, False=@n3) {parameterName=UseFakeNormalforLighting}
  ConstantBiasScale(None=DotProduct(A=CameraVectorWS(), B=@n4))
  n5 = GetMaterialAttributes(None=FunctionInput(Preview=MakeMaterialAttributes!(
     BaseColor
     Metallic
     Specular
     Roughness
     Anisotropy
     EmissiveColor
     Opacity
     OpacityMask
     Normal
     Tangent
     WorldPositionOffset
     SubsurfaceColor
     ClearCoat
     ClearCoatRoughness
     AmbientOcclusion
     Refraction
     CustomizedUVs_0
     CustomizedUVs_1
     CustomizedUVs_2
     CustomizedUVs_3
     CustomizedUVs_4
     CustomizedUVs_5
     CustomizedUVs_6
     CustomizedUVs_7
     PixelDepthOffset
     ShadingModel
     Displacement
    )) {inputType=FunctionInput_MaterialAttributes}) {attributeGetTypes [
    #1
    #2
    #3
    #4
    #5
    #6
   ]}
  n6 = CollectionParameter() {
   collection=&FortniteMaterialParameters
   parameterName=SunlightDirectionalVector
  }
  n7 = DotProduct(A=ComponentMask(None=@n6) {b=true, g=true, r=true}, B=@n4)
  n8 = CurveAtlasRowParameter(CurveTime=Clamp(
    None=ConstantBiasScale(None=@n7)
    Min=TextureSampleParameter2D() {parameterName=LX_ShadingControlTexture, texture=&T_BlacksRGB}.R
   )) {
   atlas=/Game/Characters/Shared/CurveAtlas/AthenaCharacterCurveAtlas
   curve=&/LexaShadingCurve
   defaultValue=11
   parameterName=ShadingCurve
  }
  n9 = Multiply(A=@n5.BaseColor, B=@n8.R)
  n10 = CurveAtlasRowParameter(CurveTime=MaterialFunctionCall() {materialFunction=&/MF_Phong}.Result) {
   atlas=&AthenaCharacterCurveAtlas
   curve=&LexaShadingCurve
   defaultValue=11
   parameterName=ShadingCurve
  }
  n11 = Multiply(
   A=Multiply(A=ChannelMaskParameter(None=@n10) {maskChannel=Green, parameterName=LX_SpecularCurveChannel, defaultValue=<rgba 0 1 0 0>}.R, B=@n5.BaseColor)
   B=LinearInterpolate(
    A=ScalarParameter() {parameterName=SpecHighlihgt_MinStrength}
    B=ScalarParameter() {defaultValue=2.5, parameterName=SpecHighlihgt_MaxStrength}
    Alpha=@n5.Specular
   )
  )
  n12 = Fresnel(ExponentIn=ScalarParameter() {defaultValue=4, group=HoloFoil, parameterName=BL_OutlineExponent1}, Normal=Transform(None=@n5.Normal)) {baseReflectFraction=0}
  n13 = LinearInterpolate(
   A=ScalarParameter() {defaultValue=-4, group=HoloFoil, parameterName=BL_OutlineC1}
   B=ScalarParameter() {defaultValue=12, group=HoloFoil, parameterName=BL_OutlineC2}
   Alpha=Power(Base=OneMinus(None=@n12), Exp=ScalarParameter() {defaultValue=12, group=HoloFoil, parameterName=BL_OutlineExponent2})
  )
  n14 = Multiply(A=Clamp(None=@n7), B=OneMinus(None=Saturate(None=@n13)))
  n15 = Add(
   A=StaticSwitchParameter(True=Add(A=@n9, B=@n11), False=@n9) {parameterName=UseSpecHighlight}
   B=Multiply(A=Multiply(A=@n14, B=@n5.BaseColor), B=ScalarParameter() {parameterName=RimHighlight_Strength})
  )
  n16 = Multiply(A=@n15, B=@n8.R) {constB=2}
  n17 = Saturate(None=@n16)
  n18 = SetMaterialAttributes!(
   MaterialAttributes=@n5.MaterialAttributes
   EmissiveColor=Add(
    A=@n5.EmissiveColor
    B=MaterialFunctionCall!(Color=@n17) {materialFunction=/Game/Packages/Fortress_SharedMaterials/Anime_Master_Material/MF_TODColorAdjustment}.Result
   )
   BaseColor=LinearInterpolate(A=Constant3Vector(), B=@n15, Alpha=ScalarParameter() {group=Lexa, parameterName=UseBaseColor})
   Roughness=ScalarParameter() {defaultValue=1, parameterName=Lexa_Roughness}
   Specular=ScalarParameter() {parameterName=Lexa_Specular}
   Metallic=ScalarParameter() {parameterName=Lexa_Metallic}
  ) {attributeSetTypes [
    #6
    #1
    #3
    #4
    #5
   ]}
  FunctionOutput(None=@n18)
  Multiply(
   A=@n16
   B=MaterialFunctionCall(
    Color_Morning=VectorParameter() {parameterName=LX_MorningColorTint, defaultValue=<rgba 1 1 1 1>}.RGB
    Color_Day=VectorParameter() {parameterName=LX_DayColorTint, defaultValue=<rgba 1 1 1 1>}.RGB
    Color_Evening=VectorParameter() {parameterName=LX_DayEveningTint, defaultValue=<rgba 1 1 1 1>}.RGB
    Color_Night=VectorParameter() {parameterName=LX_NightEveningTint, defaultValue=<rgba 0.5891 0.7914 1 1>}.RGB
   ) {materialFunction=&Color_Time_of_Day}.Result
  )
  Multiply(B=MaterialFunctionCall(
    Color_Morning=ScalarParameter() {defaultValue=1.5, parameterName=LX_MorningEmissiveStrength}
    Color_Day=ScalarParameter() {defaultValue=1.5, parameterName=LX_DayEmissiveStrength}
    Color_Evening=ScalarParameter() {defaultValue=1.5, parameterName=LX_EveningEmissiveStrength}
    Color_Night=ScalarParameter() {defaultValue=1, parameterName=LX_NightEmissiveStrength}
   ) {materialFunction=&Color_Time_of_Day}.Result)
  n19 = PreSkinnedPosition()
  n20 = VertexInterpolator(VS=@n19)
  n21 = Divide(
   A=FunctionInput(Preview=ShaderStageSwitch(PixelShader=@n20.PS, VertexShader=@n19)) {inputName=Preskinned}
   B=MaterialFunctionCall() {materialFunction=&ObjectLocalBounds}.Local Bounds Max
  )
  n22 = Add(A=OneMinus(None=ComponentMask(None=@n21) {b=true}), B=ScalarParameter() {defaultValue=-0.035, parameterName=HairGradientOffset})
  n23 = ConstantBiasScale(None=Sine(None=Multiply(A=@n22) {constB=6}))
  Multiply(A=@n8.G, B=Multiply(A=@n23, B=@n23))
 ]}
