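# Material function for Lexa's transformation, swapping incoming base color to a bodysuit texture as Lexa_TransformationCurve drives a reveal.
# A sphere mask on pre-skinned position, broken up by two noise textures, sets the hard swap and a glowing emissive edge.
# A static switch (Lexs_UseHairTransition) enables world position offset that pulls vertices toward headLocationSocket along the same curve.
# Time-based progress and directional hair-offset branches are left unconnected to the output.
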MaterialFunction /Game/Characters/Player/Female/Medium/Bodies/F_MED_Lexa/Materials/MF_LexaTransition
graph {nodes [
  n1 = PreSkinnedPosition()
  n2 = FunctionInput(Preview=ShaderStageSwitch(PixelShader=VertexInterpolator(VS=@n1).PS, VertexShader=@n1)) {bUsePreviewValueAsDefault=true, inputName=PreskinnedPosition}
  n3 = Divide(A=@n2, B=MaterialFunctionCall() {materialFunction=&ObjectLocalBounds}.Local Bounds Max)
  n4 = VectorParameter() {parameterName=headLocationSocket, defaultValue=<rgba 0 0 0 1>}
  n5 = Normalize(VectorInput=Subtract(A=@n4.RGB, B=WorldPosition().XYZ))
  n6 = ComponentMask(None=@n5) {g=true, r=true}
  n7 = VectorParameter() {parameterName=HairWPO_Value, defaultValue=<rgba 50 50 100 1>}
  n8 = ScalarParameter() {parameterName=Lexa_TransformationCurve}
  n9 = Multiply(A=@n8, B=ComponentMask(None=@n7.RGB) {g=true, r=true}) {constB=100}
  n10 = ComponentMask(None=@n5) {b=true}
  n11 = Multiply(A=@n8, B=@n7.B) {constB=100}
  n12 = Multiply(A=OneMinus(None=ComponentMask(None=@n3) {b=true}), B=AppendVector(A=Multiply(A=@n6, B=@n9), B=Multiply(A=@n10, B=@n11)))
  Add(A=@n12, B=Multiply(A=Constant3Vector() {constant=<rgba 0 0 -5 0>}, B=@n8))
  Add(
   A=MaterialFunctionCall() {materialFunction=&ObjectPivotPoint}.Object Pivot Location
   B=@n4.RGB
  )
  n13 = Subtract(A=@n4.RGB, B=WorldPosition().XYZ)
  Add(A=Add(A=Normalize(VectorInput=@n13), B=Constant3Vector() {constant=<rgba 1 0 -1 0>}), B=@n12)
  Constant() {r=1}
  Frac(None=Multiply(A=Time(), B=ScalarParameter() {defaultValue=0.25, parameterName=DebugSpeed}))
  Frac(None=Multiply(A=Time(), B=ScalarParameter() {parameterName=DebugSpeed}))
  n14 = GetMaterialAttributes(None=FunctionInput() {inputName=Material, inputType=FunctionInput_MaterialAttributes}) {attributeGetTypes [#1, #2]}
  n15 = TextureSampleParameter2D() {
   group=Lexa
   parameterName=BodySuitTexture
   texture=/Game/Characters/Player/Female/Medium/Bodies/F_MED_Lexa/Skins/Lexa_Transition/Textures/T_F_MED_Lexa_Transition_Body_D
  }
  n16 = SphereMask(
   A=@n2
   B=VectorParameter() {parameterName=LX_MaskLocation, defaultValue=<rgba 0 6 128 1>}.RGB
   Radius=Multiply(A=ScalarParameter() {defaultValue=128, parameterName=LX_MaskRadius}, B=@n8)
   Hardness=ScalarParameter() {parameterName=LX_MaskHardness}
  )
  n17 = TextureCoordinate()
  n18 = ComponentMask(None=VectorParameter() {parameterName=LX_OffsetNoiseTiling, defaultValue=<rgba 12 0 8 1>}.RGB) {g=true, r=true}
  n19 = Multiply(
   A=TextureSample(UVs=Multiply(A=@n17, B=@n18)) {texture=/Game/Effects/Fort_Effects/Effects/Atlas/Textures/T_Atlas_DigitalNoise_02}.R
   B=ScalarParameter() {defaultValue=0.25, parameterName=LX_EdgeLineOffset_Amt}
  )
  n20 = Multiply(
   A=TextureCoordinate()
   B=ComponentMask(None=VectorParameter() {parameterName=LX_TransitionNoiseTiling, defaultValue=<rgba 2 4 8 1>}.RGB) {g=true, r=true}
  )
  n21 = Subtract(
   A=Subtract(A=Add(A=@n16, B=@n19), B=OneMinus(None=@n8))
   B=Multiply(
    A=TextureSample(UVs=@n20) {texture=/Game/Global/Textures/Default/Noise/T_TileNoise_02}.R
    B=ScalarParameter() {defaultValue=0.5, parameterName=LX_TransitionNoiseStrength}
   )
  )
  n22 = LinearInterpolate(A=@n14.BaseColor, B=@n15.RGB, Alpha=Clamp(None=Ceil(None=@n21)))
  n23 = SphereMask(
   A=@n21
   B=Constant()
   Radius=ScalarParameter() {defaultValue=0.02, parameterName=LX_TransitionEdgeRadius}
   Hardness=ScalarParameter() {defaultValue=1, parameterName=LX_TransitionEdgeHardness}
  )
  n24 = VectorParameter() {parameterName=LX_TransitionEdge_Color, defaultValue=<rgba 1 0 0 1>}
  n25 = Power(Base=Multiply(A=Multiply(A=@n23, B=@n24.RGB), B=ScalarParameter() {defaultValue=5, parameterName=LX_TransitionEdge_GlowStrength}))
  n26 = LinearInterpolate(A=@n14.WorldPositionOffset, B=Add(A=@n14.WorldPositionOffset, B=@n13) {constB=0}, Alpha=@n8)
  FunctionOutput(None=SetMaterialAttributes!(
    MaterialAttributes=@n14.MaterialAttributes
    BaseColor=LinearInterpolate(A=@n22, B=@n25, Alpha=@n23)
    EmissiveColor=@n25
    WorldPositionOffset=StaticSwitchParameter(True=@n26, False=@n14.WorldPositionOffset) {parameterName=Lexs_UseHairTransition}
   ) {attributeSetTypes [#1, #3, #2]})
  n27 = Time()
  n28 = ScalarParameter() {defaultValue=99999997952, parameterName=LX_GameTimeFromBP}
  n29 = Saturate(None=Divide(A=Subtract(A=@n27, B=@n28), B=ScalarParameter() {defaultValue=2, parameterName=LX_TransitionDuration}))
  n30 = ScalarParameter() {parameterName=LX_TransformationDirection}
  LinearInterpolate(A=@n29, B=OneMinus(None=@n29), Alpha=@n30)
  n31 = Saturate(None=Divide(A=Subtract(A=@n27, B=@n28), B=ScalarParameter() {defaultValue=2, parameterName=LX_HairTransitionDuration}))
  LinearInterpolate(A=@n31, B=OneMinus(None=@n31), Alpha=@n30)
  Normalize(VectorInput=Subtract(
    A=WorldPosition().XYZ
    B=MaterialFunctionCall() {materialFunction=&ObjectPivotPoint}.Object Pivot Location
   ))
  ObjectPositionWS()
  OneMinus()
  TextureSample() {texture=/Game/Characters/Player/Female/Medium/Bodies/F_MED_Lexa/Textures/T_F_MED_Lexa_Body_D}
  VectorParameter() {parameterName=headLocationSocket}
 ]}
