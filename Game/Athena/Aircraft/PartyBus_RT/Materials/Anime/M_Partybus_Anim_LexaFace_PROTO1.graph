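# Prototype skeletal-mesh material, by name an anime Lexa face for the animated Party Bus, built on material attributes.
# Base character PreFX feeds cartoon shading, with switchable Lexa transition, anime eyes, Banana and flipbook mouth stages.
# HideProp scales a WPO pull toward the object position before PostFX, followed by optional outline and toon-line switches.
# Party Bus animation WPO is added last, optionally with vertex-color-masked cloth motion and blended cloth normals.
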Material /Game/Athena/Aircraft/PartyBus_RT/Materials/Anime/M_Partybus_Anim_LexaFace_PROTO1
graph {
 properties {
  bUseMaterialAttributes=true
  bUsedWithMeshDeformer=true
  bUsedWithSkeletalMesh=true
  pixelDepthOffsetMode=PDOM_Legacy
  usedShadingModels=Default Lit
 }
 nodes [
  n1 = StaticComponentMaskParameter(None=TextureSample(Tex=TextureObjectParameter() {group=Skin FX, parameterName=SkinFX_Mask}).RGB) {defaultR=true, group=Skin FX, parameterName=Flipbook Texture Channel}
  n2 = MaterialFunctionCall() {materialFunction=&BaseFN_CharacterFunction_PreFX}
  n3 = MaterialFunctionCall!(
   Preskinned=@n2.Pre skinned Local Position
   In=StaticSwitchParameter(
    True=MaterialFunctionCall!(PreskinnedPosition=@n2.Pre skinned Local Position, Material=@n2.Pre skinned Local Position) {materialFunction=/Game/Characters/Player/Female/Medium/Bodies/F_MED_Lexa/Materials/MF_LexaTransition}.Result
    False=@n2.Result
   ) {parameterName=UseLexaTransform}
  ) {materialFunction=/Game/Characters/Player/Female/Medium/Bodies/F_MED_Lexa/Materials/MF_CartoonTest}
  n4 = StaticSwitchParameter(
   True=MaterialFunctionCall!(Material=@n3.Result) {materialFunction=/Game/Characters/Player/Female/Medium/Bodies/F_MED_Lexa/Materials/MF_AnimeEyes}.Result
   False=@n3.Result
  ) {parameterName=bUseAnimeEyes}
  n5 = StaticSwitchParameter(
   True=MaterialFunctionCall!(Gmap=@n2.Gmap, Material=@n4) {materialFunction=/Game/Weapons/WeaponSkins/MaterialFunction/Banana}.Result
   False=@n4
  ) {group=Skin FX, parameterName=Use Banana}
  MaterialFunctionCall!(Alpha=@n1, Material=@n5) {materialFunction=/Game/Weapons/WeaponSkins/MaterialFunction/SubUVTextures}
  MaterialFunctionCall!(UseMaskToBrighten, SuperLevelMask=Constant(), Material=@n2.Result) {materialFunction=/Game/Characters/Player/Female/Medium/Bodies/F_MED_TNTina/Materials/MF_CartoonHatchShading}
  MaterialFunctionCall!(Material=@n4) {materialFunction=/Game/Characters/Player/Male/Medium/Bodies/M_MED_SleepyTime/Materials/MF_CheapToonEyes}
  n6 = GetMaterialAttributes(None=StaticSwitchParameter(
    True=MaterialFunctionCall!(Alpha=@n1, Material_1=@n5) {materialFunction=/Game/Characters/Player/Male/Medium/Bodies/M_MED_SleepyTime/Materials/MF_CheapFlipBookMouth}.Result
    False=@n5
   ) {group=Skin FX, parameterName=Use Sub UV texture}) {attributeGetTypes [#1]}
  n7 = Multiply(A=Subtract(A=ObjectPositionWS(), B=WorldPosition().XYZ), B=ScalarParameter() {parameterName=HideProp})
  n8 = MaterialFunctionCall(
   AO=@n2.AO
   Distance Blend=@n2.AO
   Fresnel=@n2.AO
   Gmap=@n2.AO
   Pre Skinned Local Position=@n2.AO
   Material=SetMaterialAttributes!(MaterialAttributes=@n6.MaterialAttributes, World Position Offset=Add(A=@n6.World Position Offset, B=@n7)) {attributeSetTypes [#1]}
  ) {materialFunction=&BaseFN_CharacterFunction_PostFX}
  n9 = StaticSwitchParameter(
   True=MaterialFunctionCall() {materialFunction=/Game/Characters/Player/Female/Medium/Bodies/F_MED_TNTina/Materials/MF_CartoonOutline}.Result
   False=@n8.Result
  ) {group=Cartoon, parameterName=UseCharacterCartoonOutline}
  n10 = GetMaterialAttributes(None=StaticSwitchParameter(
    True=MaterialFunctionCall!(Material=@n9) {materialFunction=/Game/Characters/Player/Male/Medium/Bodies/M_MED_Myth/Variants/Toon/Materials/MF_ToonLines}.Result
    False=@n9
   ) {group=Cartoon, parameterName=UseToonLines}) {attributeGetTypes [#2, #1]}
  n11 = MaterialFunctionCall() {materialFunction=/Game/Athena/Aircraft/PartyBus_RT/Materials/MF_PartyBusAnimationWPO}
  n12 = MaterialFunctionCall!(
   UVs
   Pan Speed
   WPO_ClothOffsetDirection
   WPO_ClothNormalIntensity
   WPO_ClothOffsetIntensity
   "Motion Mask (Defaults to the Red Vertex Color)"=ChannelMaskParameter(None=VertexColor()) {group=WPO_ClothAnimation, maskChannel=Blue, parameterName=ClothMask, defaultValue=<rgba 0 0 1 0>}.R
   WPO_ClothWPOWindTextureOffset
   Pan Direction
  ) {materialFunction=/Game/Packages/Fortress_SharedMaterials/Base_Material_Functions/Cloth}
  n13 = Add(A=@n11.Result, B=@n12.Additive WPO)
  SetMaterialAttributes!(
   MaterialAttributes=@n10.MaterialAttributes
   World Position Offset=Add(
    A=@n10.World Position Offset
    B=StaticSwitchParameter(True=@n13, False=@n11.Result) {group=WPO_ClothAnimation, parameterName=UseWPO_ClothAnimation}
   )
   Normal=StaticSwitchParameter(
    True=MaterialFunctionCall(BaseNormal=@n12.Normal, AdditionalNormal=@n10.Normal) {materialFunction=&BlendAngleCorrectedNormals}.Result
    False=@n10.Normal
   ) {group=WPO_ClothAnimation, parameterName=UseWPO_ClothAnimation}
  ) {attributeSetTypes [#1, #2]}
 ]
}
