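# Material function for the Birthday 5 backpack that layers animated, glitching confetti onto incoming material attributes.
# Two confetti textures scroll through MF_GlitchFunction on UV or optional dual-planar coordinates, mask-blended with triplanar projections of pre-skinned local position that keep confetti attached to the mesh.
# The layers composite by stepped alpha, then drive base color, flattened normal, emissive and roughness.
# MF_Birthday5_Outline processes the result, and confetti color is also output separately.
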MaterialFunction /Game/Accessories/FORT_Backpacks/Backpack_M_MED_FN_Birthday_5/Materials/MF_Birthday5_GlitchConfetti_Bus
graph {nodes [
  n1 = Multiply(A=TextureCoordinate(), B=ScalarParameter() {defaultValue=1, group=Confetti, parameterName=Tiling})
  n2 = Time()
  n3 = VectorParameter() {group=Confetti, parameterName=Direction_01, channelNames {a="", b="", g=Direction_Y, r=Direction_X}, defaultValue=<rgba 1 1 0 0>}
  n4 = MaterialFunctionCall(X=@n3.Direction_X, Y=@n3.Direction_X) {materialFunction=&MakeFloat2}
  n5 = StaticSwitchParameter(True=NamedRerouteUsage(), False=Add(A=@n1, B=Multiply(A=@n2, B=@n4.Result) {constB=0.2})) {group=Confetti, parameterName=Use Dual_Planar}
  n6 = ComponentMask(None=@n5) {r=true}
  n7 = MaterialFunctionCall!(UV=@n5) {materialFunction=/Game/Weapons/WeaponSkins/MaterialFunction/MF_GlitchFunction}
  n8 = ComponentMask(None=@n7.Glitched UV) {g=true}
  n9 = TextureSampleParameter2D(UVs=StaticSwitchParameter(True=AppendVector(A=@n6, B=@n8), False=@n7.Glitched UV) {group=Confetti, parameterName=Force_Y_Glitch}) {
   group=Confetti
   parameterName=Confetti_01
   texture=/Game/Accessories/FORT_Backpacks/Backpack_M_MED_FN_Birthday_5/Textures/TechArt/T_M_MED_FN_Birthday_5_Backpack_Glitch_Confetti_03
  }
  n10 = TextureObjectParameter() {
   group=Confetti
   parameterName=Confetti_01
   texture=&T_M_MED_FN_Birthday_5_Backpack_Glitch_Confetti_03
  }
  n11 = ScalarParameter() {defaultValue=124.4593, group=Confetti, parameterName=Confetti_Size}
  n12 = FunctionInput() {inputName=Pre-Skinned Local Position}
  n13 = MaterialFunctionCall(Float3=Divide(A=@n12, B=@n11)) {materialFunction=&BreakOutFloat3Components}
  n14 = MaterialFunctionCall(Float2=MaterialFunctionCall(UV=AppendVector(A=@n13.R, B=@n13.R)) {materialFunction=&MF_GlitchFunction}.Glitched UV) {materialFunction=&BreakOutFloat2Components}
  n15 = VectorParameter() {group=Confetti, parameterName=Direction_WPO_01, channelNames {a="", b="", g=Direction_Y, r=Direction_X}, defaultValue=<rgba 1 1 0 0>}
  n16 = Time()
  n17 = AppendVector(A=Frac(None=Multiply(A=@n15.Direction_X, B=@n16)), B=Frac(None=Multiply(A=@n15.Direction_Y, B=@n16)))
  n18 = Frac(None=Multiply(A=@n15.B, B=@n16))
  n19 = VertexInterpolator(VS=PreSkinnedNormal())
  n20 = ScalarParameter() {defaultValue=2.9375, parameterName=Contrast_Confetti}
  n21 = MaterialFunctionCall(
   TextureObject=@n10
   TextureSize=Constant3Vector() {constant=<rgba 1 1 1 0>}
   WorldPosition=Add(
    A=MaterialFunctionCall(X=@n14.R, Y=@n14.R, Z=@n13.B) {materialFunction=&MakeFloat3}.Result
    B=AppendVector(A=@n17, B=@n18)
   )
   Export Float 4=StaticBool() {value=true}
   World Space Normal=@n19.PS
   ProjectionTransitionContrast=@n20
  ) {materialFunction=&WorldAlignedTexture}
  n22 = NamedRerouteUsage()
  n23 = VectorParameter() {group=Confetti, parameterName=Direction_02, channelNames {a="", b="", g=Direction_Y, r=Direction_X}, defaultValue=<rgba -1 -1 0 0>}
  n24 = MaterialFunctionCall(X=@n23.Direction_X, Y=@n23.Direction_X) {materialFunction=&MakeFloat2}
  n25 = StaticSwitchParameter(True=NamedRerouteUsage(), False=Add(A=@n1, B=Multiply(A=@n2, B=@n24.Result))) {group=Confetti, parameterName=Use Dual_Planar}
  n26 = ComponentMask(None=@n25) {r=true}
  n27 = MaterialFunctionCall(UV=@n25) {materialFunction=&MF_GlitchFunction}
  n28 = ComponentMask(None=@n27.Glitched UV) {g=true}
  n29 = TextureSampleParameter2D(UVs=StaticSwitchParameter(True=AppendVector(A=@n26, B=@n28), False=@n27.Glitched UV) {group=Confetti, parameterName=Force_Y_Glitch}) {
   group=Confetti
   parameterName=Confetti_02
   texture=/Game/Accessories/FORT_Backpacks/Backpack_M_MED_FN_Birthday_5/Textures/TechArt/T_M_MED_FN_Birthday_5_Backpack_Glitch_Confetti_04
  }
  n30 = TextureObjectParameter() {
   group=Confetti
   parameterName=Confetti_02
   texture=&T_M_MED_FN_Birthday_5_Backpack_Glitch_Confetti_04
  }
  n31 = MaterialFunctionCall(Float3=Divide(A=@n12, B=@n11)) {materialFunction=&BreakOutFloat3Components}
  n32 = MaterialFunctionCall(Float2=MaterialFunctionCall(UV=AppendVector(A=@n31.R, B=@n31.R)) {materialFunction=&MF_GlitchFunction}.Glitched UV) {materialFunction=&BreakOutFloat2Components}
  n33 = VectorParameter() {group=Confetti, parameterName=Direction_WPO_02, channelNames {a="", b="", g=Direction_Y, r=Direction_X}, defaultValue=<rgba 1 1 0 0>}
  n34 = Time()
  n35 = AppendVector(A=Frac(None=Multiply(A=@n33.Direction_X, B=@n34)), B=Frac(None=Multiply(A=@n33.Direction_Y, B=@n34)))
  n36 = Frac(None=Multiply(A=@n33.B, B=@n34))
  n37 = MaterialFunctionCall(
   TextureObject=@n30
   TextureSize=Constant3Vector() {constant=<rgba 1 1 1 0>}
   WorldPosition=Add(
    A=MaterialFunctionCall(X=@n32.R, Y=@n32.R, Z=@n31.B) {materialFunction=&MakeFloat3}.Result
    B=AppendVector(A=@n35, B=@n36)
   )
   Export Float 4=StaticBool() {value=true}
   World Space Normal=@n19.PS
   ProjectionTransitionContrast=@n20
  ) {materialFunction=&WorldAlignedTexture}
  n38 = LinearInterpolate(A=@n9.A, B=ComponentMask(None=@n21.XYZ Texture) {a=true}, Alpha=@n22)
  n39 = LinearInterpolate(
   A=LinearInterpolate(A=@n9.RGB, B=ComponentMask(None=@n21.XYZ Texture) {b=true, g=true, r=true}, Alpha=@n22)
   B=LinearInterpolate(A=@n29.RGB, B=ComponentMask(None=@n37.XYZ Texture) {b=true, g=true, r=true}, Alpha=@n22)
   Alpha=Step(Y=@n38) {constX=0.5}
  )
  FunctionOutput(None=NamedRerouteDeclaration(None=Saturate(None=@n39)) {name=Confetti, nodeColor=<rgba 0 0.6953 1 1>}) {outputName=Base_Color}
  n40 = GetMaterialAttributes(None=NamedRerouteUsage()) {attributeGetTypes [#1]}
  n41 = NamedRerouteUsage()
  n42 = NamedRerouteUsage()
  n43 = GetMaterialAttributes(None=NamedRerouteUsage()) {attributeGetTypes [#2]}
  n44 = Multiply(A=NamedRerouteUsage(), B=ScalarParameter() {group=Confetti, parameterName=Normal_Flatten, sliderMax=1})
  n45 = GetMaterialAttributes(None=NamedRerouteUsage()) {attributeGetTypes [#3]}
  n46 = Multiply(A=NamedRerouteUsage(), B=ScalarParameter() {defaultValue=0.2, group=Confetti, parameterName=Confetti_Emissive}) {constB=20}
  n47 = NamedRerouteUsage()
  n48 = SetMaterialAttributes!(
   MaterialAttributes=NamedRerouteUsage()
   BaseColor=NamedRerouteDeclaration(None=LinearInterpolate(A=@n40.BaseColor, B=@n41, Alpha=@n42)) {name=MA_Base_Color, nodeColor=<rgba 0 0.3438 1 1>}
   Normal=NamedRerouteDeclaration(None=MaterialFunctionCall(Normal=@n43.Normal, Flatness=@n44) {materialFunction=&FlattenNormal}.Result) {name=MA_Normal, nodeColor=<rgba 0 1 0.9531 1>}
   EmissiveColor=NamedRerouteDeclaration(None=LinearInterpolate(A=@n45.EmissiveColor, B=@n46, Alpha=@n47)) {name=MA_Emissive, nodeColor=<rgba 0.3828 0 1 1>}
   Roughness=NamedRerouteUsage()
  ) {attributeSetTypes [
    #1
    #2
    #3
    #4
   ]}
  FunctionOutput(None=MaterialFunctionCall!(Material Attributes=@n48, Opacity_Mask=NamedRerouteUsage()) {materialFunction=&/MF_Birthday5_Outline}.Result) {outputName=Material_Attributes}
  MaterialFunctionCall!(In) {materialFunction=/Game/Weapons/WeaponSkins/MaterialFunction/MF_GlitchSubfunction}
  n49 = LinearInterpolate(A=@n29.A, B=ComponentMask(None=@n37.XYZ Texture) {a=true}, Alpha=@n22)
  NamedRerouteDeclaration(None=Subtract(A=Add(A=@n38, B=@n49)) {constB=0.5}) {name=Rougness_Matte, nodeColor=<rgba 0 0.2969 1 1>}
  n50 = FunctionInput() {inputName=SkinFX_Mask, inputType=FunctionInput_Vector4}
  n51 = ChannelMaskParameter(None=@n50) {group=Confetti, maskChannel=Blue, parameterName=Confetti_Mask_WorldAligned, defaultValue=<rgba 0 0 1 0>}
  NamedRerouteDeclaration(None=@n51.R) {name=Confetti_Mask2, nodeColor=<rgba 0 1 0.6484 1>}
  NamedRerouteDeclaration(None=ChannelMaskParameter(None=@n50) {group=Confetti, parameterName=Confetti_Mask_Channel}.R) {name=Confetti_Mask, nodeColor=<rgba 0.1953 0 1 1>}
  n52 = ScalarParameter() {defaultValue=1, group=Confetti, parameterName=Tiling}
  n53 = VectorParameter() {group=Confetti, parameterName=Direction_01, channelNames {a="", b="", g=Direction_Y, r=Direction_X}, defaultValue=<rgba 1 1 0 0>}
  NamedRerouteDeclaration(None=MaterialFunctionCall!(Tiling=@n52, Speed=ComponentMask(None=@n53) {g=true, r=true}, Pre-Skinned Local Position=@n12) {materialFunction=&/MF_DualPlanar}.Result) {name=Dual_Planar_01, nodeColor=<rgba 0 1 0.8359 1>}
  n54 = VectorParameter() {group=Confetti, parameterName=Direction_02, channelNames {a="", b="", g=Direction_Y, r=Direction_X}, defaultValue=<rgba -1 -1 0 0>}
  NamedRerouteDeclaration(None=MaterialFunctionCall(Tiling=@n52, Speed=ComponentMask(None=@n54) {g=true, r=true}, Pre-Skinned Local Position=@n12) {materialFunction=&MF_DualPlanar}.Result) {name=Dual_Planar_02, nodeColor=<rgba 0 1 0.8359 1>}
  n55 = GetMaterialAttributes(None=NamedRerouteUsage()) {attributeGetTypes [#4]}
  n56 = ScalarParameter() {defaultValue=0.1, group=Confetti, parameterName=Confetti_Roughness}
  n57 = NamedRerouteUsage()
  NamedRerouteDeclaration(None=LinearInterpolate(A=@n55.Roughness, B=LinearInterpolate(A=@n55.Roughness, B=@n56, Alpha=@n57), Alpha=NamedRerouteUsage())) {name=MA_Roughness, nodeColor=<rgba 0.2422 1 0 1>}
  NamedRerouteDeclaration(None=FunctionInput() {inputName=Material Attributes, inputType=FunctionInput_MaterialAttributes}) {name=Material_Attributes, nodeColor=<rgba 0.3594 0 1 1>}
  NamedRerouteDeclaration(None=FunctionInput() {inputName=Opacity_Mask, inputType=FunctionInput_Vector4}) {name=Opacity_Mask, nodeColor=<rgba 0.8047 0 1 1>}
  Reroute(None=@n51.R)
  ScalarParameter() {defaultValue=124.4593, group=Confetti, parameterName=Confetti_Size}
  Transform(None=VertexNormalWS()) {transformSourceType=TRANSFORMSOURCE_World, transformType=TRANSFORM_Local}
 ]}
