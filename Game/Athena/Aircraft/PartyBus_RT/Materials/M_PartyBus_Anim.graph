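# Animated Party Bus master material for Battle Bus skins, Nanite-enabled and built on material attributes.
# Switchable second texture set, color correction, UV1 colorization, fresnel and time-of-day-masked emissive feed the surface attributes.
# These pass through iridescent metal, optional metals LUT and birthday confetti glitch functions.
# A world-position-offset animation function follows, then optional birthday galaxy and the Creative move-tool function.
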Material /Game/Athena/Aircraft/PartyBus_RT/Materials/M_PartyBus_Anim
graph {
 properties {bUseMaterialAttributes=true, bUsedWithNanite=true, pixelDepthOffsetMode=PDOM_Legacy, usedShadingModels=Default Lit}
 nodes [
  n1 = ScalarParameter() {defaultValue=1, group=Galaxy, parameterName=GalaxyFresnel}
  n2 = TextureSampleParameter2D() {group=Textures, parameterName=FX_Mask, samplerType=SAMPLERTYPE_Masks, texture=&T_BlackMask}
  n3 = Multiply(
   A=ChannelMaskParameter(None=@n2.RGBA) {group=Galaxy, parameterName=Galaxy_FX_Mask_Channel}.R
   B=ScalarParameter() {defaultValue=1, parameterName=Galaxy_Alpha}
  )
  n4 = NamedRerouteUsage()
  n5 = NamedRerouteUsage()
  n6 = ScalarParameter() {defaultValue=1, group=Fresnel, parameterName=Fresnel_Exponent}
  n7 = Transform(None=NamedRerouteUsage())
  n8 = Add(
   A=NamedRerouteUsage()
   B=Multiply(
    A=VectorParameter() {parameterName=FresnelColor, defaultValue=<rgba 1 1 1 1>}.RGB
    B=Fresnel(ExponentIn=@n6, Normal=@n7) {baseReflectFraction=0}
   )
  )
  n9 = NamedRerouteUsage()
  n10 = MaterialFunctionCall!(
   SkinFX_Mask=NamedRerouteUsage()
   Material Attributes=MakeMaterialAttributes!(
    BaseColor=StaticSwitchParameter(True=@n8, False=@n9) {parameterName=UseFresnel}
    Metallic=NamedRerouteUsage()
    Specular=NamedRerouteUsage()
    Roughness=NamedRerouteUsage()
    Anisotropy
    EmissiveColor=NamedRerouteUsage()
    Opacity=NamedRerouteUsage()
    OpacityMask=TextureSampleParameter2D() {parameterName=OpacityMask, samplerType=SAMPLERTYPE_Masks, texture=&T_BlackMask}.G
    Normal=NamedRerouteUsage()
    Tangent
    WorldPositionOffset
    SubsurfaceColor=NamedRerouteUsage()
    ClearCoat=NamedRerouteUsage()
    ClearCoatRoughness=ScalarParameter() {parameterName=ClearCoatRoughness}
    AmbientOcclusion
    Refraction
    CustomizedUVs_0
    CustomizedUVs_1
    CustomizedUVs_2
    CustomizedUVs_3
    CustomizedUVs_4
    CustomizedUVs_5
    CustomizedUVs_6
    CustomizedUVs_7
    PixelDepthOffset
    ShadingModel
    Displacement
   )
  ) {materialFunction=&/MF_IridiscentMetal}
  n11 = MaterialFunctionCall!(In=StaticSwitchParameter(
    True=MaterialFunctionCall(Mask=@n5, Materials=@n10.Result) {materialFunction=&MF_MetalsLUT}.Result
    False=@n10.Result
   ) {parameterName=UseMatelsLUT}) {materialFunction=/Game/Environments/Props/Winter/Materials/EmissiveTextureWinterTree}
  n12 = Constant()
  n13 = VertexInterpolator(VS=PreSkinnedPosition())
  n14 = MaterialFunctionCall!(
   MA=StaticSwitchParameter(
    True=MaterialFunctionCall!(SkinFX_Mask=@n4, Material Attributes=@n11.Material Output, Opacity_Mask=@n12, Pre-Skinned Local Position=@n13.PS) {materialFunction=/Game/Accessories/FORT_Backpacks/Backpack_M_MED_FN_Birthday_5/Materials/MF_Birthday5_GlitchConfetti_Bus}.Material_Attributes
    False=@n11.Material Output
   ) {parameterName=UseBirthdayConfettiGlitch}
   Normals=NamedRerouteUsage()
  ) {materialFunction=&/MF_PartyBusAnimationWPO_WithModifiers}
  MaterialFunctionCall(MA In=StaticSwitchParameter(
    True=MaterialFunctionCall!(In=@n14.MA, Warp Mask, GalaxyAlternateFresnel=@n1, Alpha=@n3) {materialFunction=/Game/Athena/Items/Consumables/SuperTowerGrenade/Levels/GiftBox/Materials/MF_Galaxy_Birthday}.Result
    False=@n14.MA
   ) {group=Galaxy, parameterName=UseGalaxyBirthday}) {materialFunction=&MF_CreativeMoveTool}
  n15 = TextureSampleParameter2D() {group=Textures, parameterName=M, samplerType=SAMPLERTYPE_Masks, texture=&T_BlackMask}
  n16 = TextureSampleParameter2D() {group=Textures2, parameterName=M2, samplerType=SAMPLERTYPE_Masks, texture=&T_BlackMask}
  n17 = StaticSwitchParameter(True=LinearInterpolate(A=@n15.R, B=@n16.R, Alpha=NamedRerouteUsage()), False=@n15.R) {parameterName=Use2ndMaterial}
  n18 = Multiply(A=@n17, B=ScalarParameter() {defaultValue=1, parameterName=ClearCoat})
  NamedRerouteDeclaration(None=@n18) {name=ClearCoat, nodeColor=<rgba 0 1 0.3438 1>}
  n19 = ScalarParameter() {defaultValue=1, parameterName=SSS_Opacity}
  n20 = NamedRerouteUsage()
  n21 = StaticSwitchParameter(True=LinearInterpolate(A=@n19, B=ScalarParameter() {defaultValue=1, parameterName=SSS_Opacity2}, Alpha=@n20), False=@n19) {parameterName=Use2ndMaterial}
  n22 = StaticSwitchParameter(True=Multiply(A=@n17, B=@n21), False=@n21) {defaultValue=true, parameterName=UseMtexture}
  NamedRerouteDeclaration(None=@n22) {name=Opacity, nodeColor=<rgba 1 0 0.9141 1>}
  n23 = TextureSampleParameter2D() {
   group=Textures
   parameterName=SpecularMask
   samplerType=SAMPLERTYPE_Masks
   texture=/Game/Athena/Aircraft/PartyBus_RT/Textures/T_PartyBus_RT_Balloon_S
  }
  n24 = TextureSampleParameter2D() {
   group=Textures2
   parameterName=SpecularMask2
   samplerType=SAMPLERTYPE_Masks
   texture=&T_PartyBus_RT_Balloon_S
  }
  n25 = NamedRerouteUsage()
  n26 = MaterialFunctionCall(Float3=StaticSwitchParameter(True=LinearInterpolate(A=@n23.RGB, B=@n24.RGB, Alpha=@n25), False=@n23.RGB) {parameterName=Use2ndMaterial}) {materialFunction=&BreakOutFloat3Components}
  n27 = StaticSwitchParameter(
   True=LinearInterpolate(A=ScalarParameter() {parameterName=Specular_1}, B=ScalarParameter() {defaultValue=1, parameterName=Specular_2}, Alpha=@n26.R)
   False=@n26.R
  ) {parameterName=UseModifySpecular}
  NamedRerouteDeclaration(None=@n27) {name=Specular, nodeColor=<rgba 0 1 0.0156 1>}
  n28 = StaticSwitchParameter(
   True=LinearInterpolate(A=ScalarParameter() {parameterName=Metal_Low}, B=ScalarParameter() {defaultValue=1, parameterName=Metal_High}, Alpha=@n26.G)
   False=@n26.G
  ) {parameterName=UseModifyMetal}
  NamedRerouteDeclaration(None=@n28) {name=Metallic, nodeColor=<rgba 0.6172 0 1 1>}
  n29 = StaticSwitchParameter(
   True=LinearInterpolate(A=ScalarParameter() {parameterName=Roughness_Low}, B=ScalarParameter() {defaultValue=1, parameterName=Roughness_High}, Alpha=@n26.B)
   False=@n26.B
  ) {parameterName=UseModifyRoughness}
  NamedRerouteDeclaration(None=@n29) {name=Roughness, nodeColor=<rgba 0 0.0859 1 1>}
  n30 = ScalarParameter() {defaultValue=1, parameterName=SSS_Brightness}
  n31 = ScalarParameter() {defaultValue=1, parameterName=SSS_Brightness2}
  n32 = Multiply(
   A=NamedRerouteUsage()
   B=StaticSwitchParameter(True=LinearInterpolate(A=@n30, B=@n31, Alpha=@n20), False=@n30) {parameterName=Use2ndMaterial}
  )
  NamedRerouteDeclaration(None=@n32) {name=SubSurfaceColor, nodeColor=<rgba 0 0.6016 1 1>}
  n33 = ScalarParameter() {defaultValue=-0.7021, parameterName=low}
  n34 = ScalarParameter() {defaultValue=2, parameterName=high}
  n35 = Constant() {r=0.5}
  n36 = MaterialFunctionCall!(
   Sine Phase=Divide(
    A=CollectionParameter() {collection=&FortniteMaterialParameters, parameterName=TimeOfDay_0-1}
    B=MaterialFunctionCall(Multiplier=@n35) {materialFunction=&Pi}.Result
   ) {constB=24}
   Value 1=Constant() {r=1}
   Value 2=Constant()
  ) {materialFunction=/Engine/Functions/Engine_MaterialFunctions02/Utility/Sine_Remapped}
  NamedRerouteDeclaration(None=Clamp(None=LinearInterpolate(A=@n33, B=@n34, Alpha=@n36.Remapped Sine))) {name=TimeOfDayMask, nodeColor=<rgba 0.5703 0 1 1>}
  n37 = ComponentMask(None=TextureCoordinate() {coordinateIndex=1}) {r=true}
  n38 = Floor(None=Subtract(A=@n37))
  n39 = Multiply(A=Saturate(None=Floor(None=@n37)), B=OneMinus(None=Saturate(None=@n38)))
  NamedRerouteDeclaration(None=@n39) {name=2ndUVChannelMask, nodeColor=<rgba 0 0.5312 1 1>}
  n40 = TextureSampleParameter2D() {group=Textures, parameterName=Emissive, texture=/Game/Athena/Aircraft/PartyBus_RT/Textures/T_PartyBus_RT_Balloon_E}
  n41 = StaticSwitchParameter(
   True=LinearInterpolate(
    A=@n40.RGB
    B=TextureSampleParameter2D() {group=Textures2, parameterName=Emissive2, texture=&T_PartyBus_RT_Balloon_E}.RGB
    Alpha=NamedRerouteUsage()
   )
   False=@n40.RGB
  ) {parameterName=Use2ndMaterial}
  n42 = MaterialFunctionCall(Float3=@n41) {materialFunction=&BreakOutFloat3Components}
  n43 = Add(
   A=Multiply(A=VectorParameter() {parameterName=EmissiveColorRedChannel, defaultValue=<rgba 0 0 0 1>}.RGB, B=@n42.R)
   B=Multiply(A=VectorParameter() {parameterName=EmissiveColorBlueChannel, defaultValue=<rgba 0 0 0 1>}.RGB, B=@n42.B)
  )
  n44 = Multiply(
   A=ScalarParameter() {defaultValue=1, parameterName=EmissiveBrightness}
   B=Power(
    Base=StaticSwitchParameter(True=@n43, False=@n41) {parameterName=UseEmissiveColorChannels}
    Exp=ScalarParameter() {defaultValue=1, parameterName=EmissiveContrast}
   )
  )
  n45 = Multiply(A=@n44, B=NamedRerouteUsage())
  NamedRerouteDeclaration(None=StaticSwitchParameter(True=StaticSwitchParameter(True=@n45, False=@n44) {defaultValue=true, parameterName=UseTODEmissiveMask}, False=Constant()) {defaultValue=true, parameterName=UseEmissive}) {name=Emissive, nodeColor=<rgba 0 0.25 1 1>}
  n46 = TextureSample(UVs=TextureCoordinate() {coordinateIndex=1}) {texture=/Game/Athena/Aircraft/PartyBus_Birthday/Textures/T_Partybus_Birthday_Colors_9}
  n47 = ScalarParameter() {group=ColorCorrection, parameterName=BaseColorHueShift}
  n48 = TextureSampleParameter2D() {group=Textures, parameterName=BaseColor, texture=/Game/Athena/Aircraft/PartyBus_RT/Textures/T_PartyBus_RT_Balloon_D}
  n49 = StaticSwitchParameter(
   True=LinearInterpolate(
    A=@n48.RGB
    B=TextureSampleParameter2D() {group=Textures2, parameterName=BaseColor2, texture=&T_PartyBus_RT_Balloon_D}.RGB
    Alpha=NamedRerouteUsage()
   )
   False=@n48.RGB
  ) {parameterName=Use2ndMaterial}
  n50 = StaticSwitchParameter(
   True=Multiply(
    A=VectorParameter() {group=ColorCorrection, parameterName=BaseColorTint, defaultValue=<rgba 1 1 1 1>}.RGB
    B=MaterialFunctionCall(Hue Shift Percentage=@n47, Texture=@n49) {materialFunction=&HueShift}.Result
   )
   False=@n49
  ) {parameterName=UseColorCorrection}
  NamedRerouteDeclaration(None=StaticSwitchParameter(True=Multiply(A=@n46.RGB, B=@n50), False=@n50) {parameterName=UseColorization}) {name=BaseColor, nodeColor=<rgba 0.0547 0 1 1>}
  NamedRerouteDeclaration(None=TextureSampleParameter2D() {parameterName=FX_Mask, samplerType=SAMPLERTYPE_Masks, texture=&T_BlackMask}.RGBA) {name=SkinFX_Mask, nodeColor=<rgba 0 1 0.8828 1>}
  n51 = TextureSampleParameter2D() {
   group=Textures
   parameterName=Normal
   samplerType=SAMPLERTYPE_Normal
   texture=/Game/Athena/Aircraft/PartyBus_RT/Textures/T_PartyBus_RT_Balloon_N
  }
  n52 = TextureSampleParameter2D() {
   group=Textures2
   parameterName=Normal2
   samplerType=SAMPLERTYPE_Normal
   texture=&T_PartyBus_RT_Balloon_N
  }
  n53 = NamedRerouteUsage()
  n54 = NamedRerouteDeclaration(None=StaticSwitchParameter(True=LinearInterpolate(A=@n51.RGB, B=@n52.RGB, Alpha=@n53), False=@n51.RGB) {parameterName=Use2ndMaterial}) {name=NormalTextures, nodeColor=<rgba 0.125 0 1 1>}
  NamedRerouteDeclaration(None=@n54) {name=Normal, nodeColor=<rgba 0 0.0859 1 1>}
  Reroute(None=@n39)
  Reroute(None=@n18)
  Reroute(None=@n22)
  Reroute(None=@n27)
  Reroute(None=@n28)
  Reroute(None=@n29)
  Reroute(None=@n32)
  n55 = Reroute() {desc=Cloth Normal}
  StaticSwitchParameter(
   True=ShadingPathSwitch(
    Default=MaterialFunctionCall(BaseNormal=@n54, AdditionalNormal=@n55) {materialFunction=&BlendAngleCorrectedNormals}.Result
    Mobile=@n54
   ) {desc=Migrated from FeatureLevelSwitch}
   False=@n54
  ) {group=WPO_ClothAnimation, parameterName=UseWPO_ClothAnimation}
 ]
}
